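# Material function that computes an alpha erosion (dissolve) mask from an RGBA texture input, in the crossbow launcher FX folder.
# A selectable channel feeds SmoothStep, with Erosion Amount as lower edge and Erosion Contrast narrowing the band.
# An influence parameter blends from optional base mask to eroded result, optionally multiplied by that mask.
# Static switches bypass erosion to 1 or "invert" values, which actually multiplies the channel by its inverse.
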MaterialFunction /MoonFlaxWeaponGameplay/Items/Weapons/Crossbow/FX/LauncherFX/MF_AlphaErosion_Complex
graph {
 properties {bExposeToLibrary=true, description=Erosion Function, previewBlendMode=BLEND_Translucent}
 nodes [
  n1 = FunctionInput() {inputName="Erosion Texture (RGBA)", inputType=FunctionInput_Vector4}
  n2 = ChannelMaskParameter(None=@n1) {group=Erotion, maskChannel=Alpha, parameterName=Base Mask Channel, sortPriority=4, defaultValue=<rgba 0 0 0 1>}
  n3 = FunctionInput() {inputName=Erosion Amount, inputType=FunctionInput_Scalar, sortPriority=2}
  n4 = FunctionInput() {inputName=Erosion Contrast, inputType=FunctionInput_Scalar, sortPriority=2, previewValue=<0.5 0.5 0.5 0.5>}
  n5 = ChannelMaskParameter(None=@n1) {group=Erotion, parameterName=Erosion Channel, sortPriority=-2}
  n6 = OneMinus(None=@n5.R)
  n7 = MaterialFunctionCall(
   Alpha=StaticSwitchParameter(True=Multiply(A=@n6, B=@n5.R), False=@n5.R) {group=Erotion, parameterName=Invert Erosion Values}
   Min=@n3
   Max=Add(A=@n3, B=OneMinus(None=@n4))
  ) {materialFunction=&SmoothStep}
  n8 = LinearInterpolate(
   A=StaticSwitchParameter(True=@n2.R, False=Constant() {r=1}) {group=Erotion, parameterName=Use Base Mask, sortPriority=3}
   B=StaticSwitchParameter(True=Multiply(A=@n2.R, B=@n7.Result), False=@n7.Result) {group=Erotion, parameterName=Mutiply by Base Alpha, sortPriority=3}
   Alpha=ScalarParameter() {defaultValue=1, group=Erotion, parameterName=Erosion Influence, sortPriority=-3}
  )
  FunctionOutput(None=StaticSwitchParameter(True=Saturate(None=@n8), False=Constant() {r=1}) {defaultValue=true, group=Erotion, parameterName=Erosion, sortPriority=-10}) {outputName=Color}
 ]
}
