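# Translucent lit material that fakes a projected Moonflax crossbow goo-splash decal on Niagara mesh particles.
# Scene position behind the translucency is projected in instance space onto axes built from a dynamic-parameter direction, giving bounds-normalized splash UVs.
# Texture red shades particle color, green drives emissive highlights, and alpha is eroded by distance from particle center relative to Range.
# A smoothstepped direction dot product removes back-side projection from opacity.
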Material /MoonFlaxWeaponGameplay/Items/Weapons/Crossbow/FX/LauncherFX/M_Crossbow_Moonflax_Decal
graph {
 properties {bUsedWithNiagaraMeshParticles=true, blendMode=BLEND_Translucent, pixelDepthOffsetMode=PDOM_Legacy, usedShadingModels=Default Lit}
 nodes [
  n1 = TransformPosition(None=MaterialFunctionCall() {materialFunction=&WorldPositionBehindTranslucency}.Result) {transformSourceType=TRANSFORMPOSSOURCE_World, transformType=TRANSFORMPOSSOURCE_Instance}
  n2 = DynamicParameter() {defaultValue=<rgba 0 0 -1 150>, paramNames [Projection X, Projection Y, Projection Z, Range]}
  n3 = Transform(None=Normalize(VectorInput=MaterialFunctionCall(R=@n2.Projection X, G=@n2.Projection X, B=@n2.Projection X) {materialFunction=&AppendMany}.RGB)) {transformSourceType=TRANSFORMSOURCE_World, transformType=TRANSFORM_Instance}
  n4 = CrossProduct(A=@n3, B=Constant3Vector() {constant=<rgba 0 0 1 0>})
  n5 = CrossProduct(A=@n4, B=@n3)
  n6 = AppendVector(A=DotProduct(A=@n1, B=Normalize(VectorInput=@n4)), B=DotProduct(A=@n1, B=Normalize(VectorInput=@n5)))
  n7 = ComponentMask(None=MaterialFunctionCall() {materialFunction=&ObjectLocalBounds}.Local Bounds Size) {g=true, r=true}
  n8 = Divide(A=Add(A=Divide(A=@n6, B=@n7)))
  n9 = ScalarParameter() {defaultValue=1.5, parameterName=Splash Texture Scale, sliderMax=2}
  n10 = TextureSampleParameter2D(UVs=Clamp(None=MaterialFunctionCall(UVs=@n8, Texture Scale=@n9) {materialFunction=&ScaleUVsByCenter}.UVs)) {parameterName=Splash Texture, texture=&/T_Crossbow_Moonflax_Goo_Splash_Alpha_02}
  MaterialFunctionCall(Emissive=Multiply(A=@n10.G, B=ScalarParameter() {defaultValue=0.5, parameterName=Highlights, sliderMax=1})) {materialFunction=&MF_FX_EmissiveAdjustments}
  n11 = ParticleColor()
  Multiply(A=Power(Base=@n10.R, Exp=ScalarParameter() {defaultValue=1, parameterName=Shadow Contrast, sliderMax=3}), B=@n11.RGB)
  n12 = ScalarParameter() {defaultValue=0.5, parameterName=Smooth Step Opacity Max, sliderMax=1}
  n13 = Subtract(
   A=MaterialFunctionCall() {materialFunction=&WorldPositionBehindTranslucency}.Result
   B=ParticlePositionWS()
  )
  n14 = Divide(
   A=Divide(
    A=Length(None=@n13) {is=class MaterialExpressionLength}
    B=DynamicParameter() {defaultValue=<rgba 0 0 -1 150>, paramNames [Projection X, Projection Y, Projection Z, Range]}.Range
   )
   B=MaterialFunctionCall() {materialFunction=&ObjectScale}.Scale X
  )
  n15 = MaterialFunctionCall!(
   "Erosion Texture (RGBA)"=@n10.A
   Erosion Contrast=Constant()
   Erosion Amount=Power(Base=Saturate(None=@n14), Exp=ScalarParameter() {defaultValue=3, parameterName=Edge Fade, sliderMax=5})
  ) {materialFunction=&/MF_AlphaErosion_Complex}
  n16 = DynamicParameter() {defaultValue=<rgba 0 0 -1 150>, paramNames [Projection X, Projection Y, Projection Z, Range]}
  n17 = DotProduct(
   A=Normalize(VectorInput=MaterialFunctionCall(R=@n16.Projection X, G=@n16.Projection X, B=@n16.Projection X) {materialFunction=&AppendMany}.RGB)
   B=Normalize(VectorInput=@n13)
  )
  Multiply(
   A=Multiply(A=@n11.A, B=SmoothStep(Max=@n12, Value=@n15.Color) {constMax=0.5})
   B=SmoothStep(
    Min=ScalarParameter() {parameterName=Backface Culling, sliderMax=-1}
    Max=ScalarParameter() {parameterName=Backface Culling Max}
    Value=Multiply(A=@n17) {constB=-1}
   ) {constMax=0, constMin=-0.5, desc=For removing projection from behind}
  )
 ]
}
