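# Translucent deferred decal material for the Moonflax crossbow launcher's scorch mark: a dark goo-splash shape with a glowing, animated emissive.
# Two splash masks multiply with two-layer panning noise, then a power curve and edge-to-center color blend shape the glow before HDR remapping.
# Decal lifetime opacity drives the glow's fade and changing exponent, with an offset lerp for separate timing from the mark's opacity.
# Static switches toggle emissive and choose the opacity mask texture.
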Material /MoonFlaxWeaponGameplay/Items/Weapons/Crossbow/FX/LauncherFX/M_Crossbow_Moonflax_Decal_ScorchMark
graph {
 properties {
  bAutomaticallySetUsageInEditor=false
  blendMode=BLEND_Translucent
  materialDomain=MD_DeferredDecal
  pixelDepthOffsetMode=PDOM_Legacy
  shadingModel=MSM_Unlit
  usedShadingModels=Unlit
 }
 nodes [
  n1 = DecalLifetimeOpacity()
  n2 = Clamp(None=LinearInterpolate(A=ScalarParameter() {defaultValue=-3, parameterName=EmissiveFadeOffset}, Alpha=Clamp(None=@n1.Opacity)) {constA=-7})
  n3 = ScalarParameter() {parameterName=DecalAffectEmissive_Linner}
  n4 = TextureSampleParameter2D(UVs=Add(A=TextureCoordinate() {uTiling=2, vTiling=2}) {constB=-0.5}) {parameterName=EmissiveTextureA, texture=&/T_Crossbow_MoonflaxGoo_Splash_Alpha_01}
  n5 = ScalarParameter() {defaultValue=1, parameterName=NoiseTile}
  n6 = TextureCoordinate() {uTiling=3, vTiling=3}
  n7 = Multiply(A=Time(), B=ScalarParameter() {defaultValue=0.25, parameterName=PanSpeed})
  n8 = TextureSample(UVs=Panner(Coordinate=Multiply(A=@n5, B=@n6), Time=@n7) {speedX=0.005, speedY=-0.02}) {samplerType=SAMPLERTYPE_Masks, texture=/Game/Global/Textures/Default/Noise/T_CubeNoise}
  n9 = Panner(Coordinate=Multiply(A=@n5, B=TextureCoordinate() {uTiling=4, vTiling=4}), Time=@n7) {speedX=-0.015, speedY=0.03}
  n10 = Multiply(
   A=@n8.R
   B=Multiply(A=@n8.B, B=TextureSample(UVs=@n9) {samplerType=SAMPLERTYPE_Masks, texture=&T_CubeNoise}.B)
  )
  n11 = Multiply(A=Multiply(A=@n10, B=@n10) {constB=2}, B=ScalarParameter() {defaultValue=1, parameterName=NoiseMult}) {constB=4}
  n12 = ScalarParameter() {parameterName=NoiseAdd}
  n13 = TextureSampleParameter2D(UVs=Multiply(A=TextureCoordinate(), B=ScalarParameter() {defaultValue=1, parameterName=EmissiveTextureB_Tiling})) {parameterName=EmissiveTextureB, texture=&T_Crossbow_MoonflaxGoo_Splash_Alpha_01}
  n14 = Multiply(A=Multiply(A=@n4.R, B=Add(A=@n11, B=@n12) {constB=0.7}), B=@n13.R)
  n15 = ScalarParameter() {defaultValue=14, parameterName=EndPower}
  n16 = ScalarParameter() {defaultValue=1, parameterName=StartPower}
  n17 = LinearInterpolate(A=@n2, B=@n1.Opacity, Alpha=@n3)
  n18 = Multiply(
   A=LinearInterpolate(A=Power(Base=@n2) {constExponent=0.6}, B=@n1.Opacity, Alpha=@n3)
   B=Power(Base=Saturate(None=@n14), Exp=LinearInterpolate(A=@n15, B=@n16, Alpha=@n17) {constA=14})
  )
  n19 = Multiply(
   A=LinearInterpolate(
    A=VectorParameter() {group=Emissive, parameterName=EmissiveEdge, defaultValue=<rgba 0.8 0.1474 0.0516 1>}.RGB
    B=VectorParameter() {group=Emissive, parameterName=EmissiveCenter, defaultValue=<rgba 1.5 0.1871 1.3198 1>}.RGB
    Alpha=Clamp(None=@n18)
   )
   B=@n18
  )
  n20 = Clamp(None=@n19)
  MaterialFunctionCall(Emissive=StaticSwitchParameter(
    True=MaterialFunctionCall(LDR Input=@n20, HDR Tint=@n19) {materialFunction=&DeriveHDRfromLDR}.RemappedRange
    False=Constant() {r=1}
   ) {defaultValue=true, group=Timestamp, parameterName=Use Emissive?}) {materialFunction=&MF_FX_EmissiveAdjustments}
  n21 = Multiply(A=@n1.Opacity, B=ScalarParameter() {defaultValue=1, parameterName=FadeDelay})
  n22 = TextureSampleParameter2D() {parameterName=DecalTexture, texture=&T_Crossbow_MoonflaxGoo_Splash_Alpha_01}
  n23 = TextureSampleParameter2D() {parameterName=EmissiveTextureA, texture=&T_Crossbow_MoonflaxGoo_Splash_Alpha_01}
  n24 = StaticSwitchParameter(True=@n22.G, False=@n23.G) {defaultValue=true, parameterName=UseDecalTexture}
  Multiply(A=LinearInterpolate(B=Saturate(None=@n21), Alpha=ScalarParameter() {parameterName=DecalAffectOpacity}) {constA=1}, B=@n24)
  Multiply(A=Multiply(A=@n24, B=ScalarParameter() {defaultValue=1, parameterName=OpacityMult}), B=DecalLifetimeOpacity().Opacity)
  ScalarParameter() {parameterName=Metallic}
  ScalarParameter() {parameterName=Specular}
  ScalarParameter() {defaultValue=1, parameterName=Roughness}
  VectorParameter() {parameterName=DecalColor, defaultValue=<rgba 0 0 0 1>}
 ]
}
